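# Default Lit, Nanite-enabled base material, despite its MI_ prefix, for the animated Office_ShadowScreen_A prop's emblem display.
# A time-driven rotator turns one emblem sample, which is tinted, dimmed and added to a SocietyScreen mask texture.
# A second emblem sample's red channel blends that result toward a dark color to form EmblemScreen, used by an "Emblem?" switch.
# Named reroutes provide cropped UVs and UV_Spin, which fakes spinning by time-scaling centered U through a sine.
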Material /Game/Environments/Helios/Props/Office/Office_ShadowScreen_A/Materials/MI_Office_ShadowScreen_A_Animated
graph {
 properties {bUsedWithNanite=true, pixelDepthOffsetMode=PDOM_Legacy, usedShadingModels=Default Lit}
 nodes [
  n1 = MaterialFunctionCall(UVs=NamedRerouteUsage(), Rotation Center=Constant2Vector() {g=0.5, r=0.5}, "Rotation Angle (0-1)"=Multiply(A=Time()) {constB=0.04}) {materialFunction=&CustomRotator}
  n2 = VectorParameter() {parameterName=Param_1, defaultValue=<rgba 1 1 1 1>}
  n3 = VectorParameter() {parameterName=Param_2, defaultValue=<rgba 0.0625 0.0625 0.0625 1>}
  n4 = NamedRerouteUsage()
  n5 = Multiply(
   A=Multiply(
    A=TextureSampleParameter2D(UVs=@n1.Rotated Values) {parameterName=Emblem, texture=/Game/Environments/Helios/Props/Office/Office_ShadowScreen_A/Textures/TV_Shadow_Icon}.R
    B=LinearInterpolate(A=@n2.RGB, B=@n3.RGB, Alpha=@n4)
   )
   B=ScalarParameter() {defaultValue=0.1}
  )
  n6 = TextureSample() {samplerType=SAMPLERTYPE_Masks, texture=/Game/Environments/Helios/Props/Office/Office_SocietyScreen_A/Textures/T_Office_SocietyScreen_A_A_M}
  n7 = Multiply(A=@n6.RGB)
  n8 = NamedRerouteUsage()
  n9 = NamedRerouteDeclaration(None=LinearInterpolate(
    A=Add(A=@n5, B=@n7)
    B=VectorParameter() {parameterName=Param_5, defaultValue=<rgba 0.0052 0.0035 0.0026 1>}.RGB
    Alpha=TextureSampleParameter2D(UVs=@n8) {parameterName=Emblem, texture=&TV_Shadow_Icon}.R
   )) {name=EmblemScreen, nodeColor=<rgba 1 0.4453 0 1>}
  Add(A=@n9, B=@n7)
  Constant()
  LinearInterpolate(
   A=VectorParameter() {parameterName=Param_3, defaultValue=<rgba 0 0 0 1>}.RGB
   B=VectorParameter() {parameterName=Param_4, defaultValue=<rgba 0.0677 0.0677 0.0677 1>}.RGB
   Alpha=NamedRerouteUsage()
  )
  n10 = MaterialFunctionCall(Float2=NamedRerouteUsage()) {materialFunction=&BreakOutFloat2Components}
  n11 = Frac(None=Multiply(A=Time()) {constB=0.3})
  n12 = Clamp(None=Abs(None=Subtract(A=@n11) {constB=0.5})) {maxDefault=0.49, minDefault=0.01}
  n13 = If(A=@n11, "A > B"=@n12, "A == B"=@n12, "A < B"=Multiply(A=@n12) {constB=-1}) {constB=0.5}
  n14 = Divide(A=Subtract(A=@n10.R) {constB=0.5}, B=Sine(None=Add(A=@n13) {constB=0.5}))
  NamedRerouteDeclaration(None=MaterialFunctionCall(X=Add(A=@n14) {constB=0.5}, Y=@n10.G) {materialFunction=&MakeFloat2}.Result) {name=UV_Spin, nodeColor=<rgba 1 0.0234 0 1>}
  NamedRerouteDeclaration(None=@n9) {name=debug, nodeColor=<rgba 0.8984 1 0 1>}
  n15 = VectorParameter() {defaultValue=<rgba 0.3 0.13 0.7 0.87>}
  NamedRerouteDeclaration(None=MaterialFunctionCall!(
    UVs=TextureCoordinate()
    TextureIn
    UpperLeftCorner=AppendVector(A=@n15.R, B=@n15.R)
    LowerRightCorner=AppendVector(A=@n15.B, B=@n15.B)
    ExportFloat4
   ) {materialFunction=/Engine/Functions/Engine_MaterialFunctions01/Texturing/TextureCropping}.Crop UVs) {name=CroppedUV, nodeColor=<rgba 0 0.9062 1 1>}
  n16 = VectorParameter() {parameterName=Param2, defaultValue=<rgba 0.1 -0.4 0.9 1.4>}
  NamedRerouteDeclaration(None=MaterialFunctionCall(UVs=TextureCoordinate(), UpperLeftCorner=AppendVector(A=@n16.R, B=@n16.R), LowerRightCorner=AppendVector(A=@n16.B, B=@n16.B)) {materialFunction=&TextureCropping}.Crop UVs) {name=CroppedUV_BG, nodeColor=<rgba 0 0.9062 1 1>}
  n17 = MaterialFunctionCall(Float2=TextureCoordinate()) {materialFunction=&BreakOutFloat2Components}
  NamedRerouteDeclaration(None=@n17.R) {name=U, nodeColor=<rgba 1 0.1875 0 1>}
  NamedRerouteDeclaration(None=@n17.G) {name=V, nodeColor=<rgba 0 1 0.8359 1>}
  StaticSwitchParameter()
  StaticSwitchParameter(True=@n9) {parameterName=Emblem?}
  TextureSample() {samplerType=SAMPLERTYPE_Masks, texture=/Game/Environments/Helios/Props/Office/Office_SocietyScreen_A/Textures/T_Office_SocietyScreen_B_A_S}
 ]
}
