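# Default Lit material for an animated office Shadow screen, showing a pulsing Shadow icon over a mask texture borrowed from the Society screen.
# Two samples of the Emblem texture scale around an offset pivot on separate sine-over-time pulses, with UVs clipped by TextureCropping.
# The inverted samples multiply with the dimmed mask, then pass through the shared time-of-day color curve and EyeAdaptationInverse to compensate for exposure.
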Material /Game/Environments/Helios/Props/Office/Office_ShadowScreen_A/Materials/MI_Office_ShadowScreen_B_Animated
graph {
 properties {bUsedWithNanite=true, pixelDepthOffsetMode=PDOM_Legacy, usedShadingModels=Default Lit}
 nodes [
  Constant()
  n1 = TextureSample() {samplerType=SAMPLERTYPE_Masks, texture=/Game/Environments/Helios/Props/Office/Office_SocietyScreen_A/Textures/T_Office_SocietyScreen_B_A_M}
  n2 = Constant2Vector() {g=-0.4, r=-0.65}
  n3 = Add(A=TextureCoordinate(), B=@n2)
  n4 = Add(A=Multiply(A=MaterialFunctionCall() {materialFunction=&DebugTimeSine}.Result) {constB=0.15})
  n5 = Multiply(A=@n2) {constB=-1}
  n6 = VectorParameter() {defaultValue=<rgba 0.51 0.03 0.83 0.59>}
  n7 = MaterialFunctionCall!(
   UVs=Add(A=Multiply(A=@n3, B=@n4), B=@n5) {constB=0.65}
   TextureIn
   UpperLeftCorner=AppendVector(A=@n6.R, B=@n6.R)
   LowerRightCorner=AppendVector(A=@n6.B, B=@n6.B)
   ExportFloat4
  ) {materialFunction=/Engine/Functions/Engine_MaterialFunctions01/Texturing/TextureCropping}
  n8 = Multiply(
   A=Multiply(A=@n1.RGB) {constB=0.6}
   B=OneMinus(None=TextureSampleParameter2D(UVs=@n7.Crop UVs) {parameterName=Emblem, texture=/Game/Environments/Helios/Props/Office/Office_ShadowScreen_A/Textures/TV_Shadow_Icon}.RGB)
  )
  n9 = Time()
  n10 = Sine(None=Multiply(A=Add(A=@n9) {constB=-0.3}))
  n11 = Add(A=Multiply(A=ConstantBiasScale(None=@n10)) {constB=0.15})
  n12 = VectorParameter() {parameterName=Param2, defaultValue=<rgba 0.49 0.01 0.85 0.61>}
  n13 = MaterialFunctionCall(
   UVs=Add(A=Multiply(A=@n3, B=@n11), B=@n5)
   UpperLeftCorner=AppendVector(A=@n12.R, B=@n12.R)
   LowerRightCorner=AppendVector(A=@n12.B, B=@n12.B)
  ) {materialFunction=&TextureCropping}
  n14 = Add(A=OneMinus(None=TextureSampleParameter2D(UVs=@n13.Crop UVs) {parameterName=Emblem, texture=&TV_Shadow_Icon}.RGB)) {constB=0.5}
  EyeAdaptationInverse(
   LightValueInput=MaterialFunctionCall!(Color=Multiply(A=@n8, B=@n14)) {materialFunction=/Game/Packages/Fortress_SharedMaterials/Anime_Master_Material/MF_TODColorCurve}.Result
   AlphaInput=Constant() {r=1}
  )
  MaterialFunctionCall() {materialFunction=&DebugTimeSine}
  NamedRerouteUsage()
  TextureSample() {samplerType=SAMPLERTYPE_Masks, texture=/Game/Environments/Helios/Props/Office/Office_SocietyScreen_A/Textures/T_Office_SocietyScreen_B_A_S}
 ]
}
